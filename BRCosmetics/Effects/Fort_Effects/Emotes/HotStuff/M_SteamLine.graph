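# Translucent unlit material for the steam line in the HotStuff emote, built for Niagara ribbons and beam trails.
# Two soft smoke-trail samples, mapped by object-relative world height and second-UV V, pan on a particle-driven time and multiply together.
# Dynamic parameters set both tiling sizes and alpha, while opacity fades along ribbon U by one minus U squared.
# Particle color drives emissive and alpha; depth fade softens intersections except on mobile.
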Material /BRCosmetics/Effects/Fort_Effects/Emotes/HotStuff/M_SteamLine
graph {
 properties {
  bUsedWithBeamTrails=true
  bUsedWithNiagaraRibbons=true
  blendMode=BLEND_Translucent
  pixelDepthOffsetMode=PDOM_Legacy
  shadingModel=MSM_Unlit
  usedShadingModels=Unlit
 }
 nodes [
  n1 = ParticleColor()
  MaterialFunctionCall(Emissive=@n1.RGB) {materialFunction=&MF_FX_EmissiveAdjustments}
  n2 = ComponentMask(None=Subtract(A=WorldPosition().XYZ, B=ObjectPositionWS())) {b=true}
  n3 = DynamicParameter() {defaultValue=<rgba 1 18 25 1>, paramNames [Time,U_Size2,U_Size,Alpha]}
  n4 = ComponentMask(None=TextureCoordinate() {coordinateIndex=1}) {g=true}
  n5 = AppendVector(A=Divide(A=@n2, B=@n3.U_Size2), B=@n4)
  n6 = AppendVector(A=Divide(A=@n2, B=@n3.U_Size), B=@n4)
  n7 = Multiply(
   A=TextureSample(UVs=Panner(Coordinate=@n5, Time=@n3.Time) {speedX=1}) {samplerType=SAMPLERTYPE_Grayscale, texture=&T_Smoke_Trail_Soft}.R
   B=TextureSample(UVs=Panner(Coordinate=@n6, Time=@n3.Time) {speedX=1}) {samplerType=SAMPLERTYPE_Grayscale, texture=&T_Smoke_Trail_Soft}.R
  )
  n8 = ComponentMask(None=TextureCoordinate()) {r=true}
  n9 = Multiply(A=Multiply(A=@n1.A, B=Multiply(A=@n7, B=@n3.Alpha)), B=OneMinus(None=Multiply(A=@n8, B=@n8)))
  ShadingPathSwitch(Default=DepthFade(Opacity=@n9, FadeDistance=ScalarParameter() {defaultValue=20, parameterName=DepthFadeDistance}), Mobile=@n9) {desc=Migrated from FeatureLevelSwitch}
 ]
}
