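# Niagara effect for the HotStuff emote, pairing a rising ribbon trail with a short burst of sprite streaks.
# Trail spawns ribbon points gated on a two-second age check; they drift upward, widen and fade while age-driven curves animate material parameters and overall alpha.
# Streaks2 bursts a few short-lived, faint sprites that shrink and fade, re-randomizing their sub-image in particle update for flicker.
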NiagaraSystem /BRCosmetics/Effects/Fort_Effects/Emotes/HotStuff/NS_Emote_HotStuff
dependencies {usedDataInterfaces [&NiagaraDataInterfaceCurve], usedDynamicInputs [], usedModules [&SystemState]}
emitters {
 Streaks2 {
  emitterUpdateScript [
   &EmitterLifeCycle(MaxLoopCount=1, NextLoopDelay=0.15, NextLoopDuration=1) {enabled=false}
   &EmitterState()
   &SpawnBurst_Instantaneous(Spawn Count=5, Use Spawn Probability=false)
  ]
  particleSpawnScript [
   /Niagara/Modules/Spawn/Initialization/InitializeParticle(
    Lifetime=&UniformRangedFloat(Minimum=0.2, Maximum=0.3)
    Module.Write Color=false
    Module.Write Lifetime=true
    Module.Write Mass=false
    Module.Write Material Random=true
    Module.Write Position=false
    Module.Write Scale=false
    Module.Write SpriteRotation=false
    Module.Write SpriteSize=true
    Position=&SimulationPosition()
    Sprite Size=&UniformRangedVector2D(Minimum=<4 15>, Maximum=<2.5 7>)
   )
   &AddVelocity(Velocity=/Niagara/DynamicInputs/Random/RandomVector(
     Vector Scale:NiagaraFloat=&UniformRangedFloat(Minimum=34, Maximum=72)
     Constraint Vector:Vector3f=<0 0 1>
     Randomness Mode:ENiagaraRandomnessMode=::NewEnumerator0
     Random Seed:NiagaraInt32=0
     Fixed Random Seed:NiagaraBool=false
    ))
  ]
  particleUpdateScript [
   &ParticleState() {module=UpdateAge}
   &ScaleSpriteSize(Scale Factor=&Vector2DFromFloat(Value=&FloatFromCurve(FloatCurve=DataInterface() {propertyValues {Curve {keys [
          <at -0.0046 1.0113 RCIM_Cubic>
          {arriveTangent=-1.0036, interpMode=RCIM_Cubic, leaveTangent=-1.0036, time=0.7001, value=0.872}
          <at 1.0011 0.002>
         ]}}}))) {module=SpriteSizeScale}
   /SetVariables(Particles.SubImageIndex=&UniformRangedFloat(Minimum=0, Maximum=3)) {isSetParameters=true}
   &Color(
    Color=<rgba 1 1 1 0.9>
    Scale Alpha=&FloatFromCurve(FloatCurve=DataInterface() {propertyValues {Curve {keys [{arriveTangent=-0.0654, interpMode=RCIM_Cubic, leaveTangent=-0.0654, tangentMode=RCTM_User, time=0, value=0.04}, <at 1.0011 0.002>]}}})
   )
   &SolveForcesAndVelocity(Clamp Velocity=false, Limit Acceleration=false)
  ]
  rendererClasses [&NiagaraSpriteRendererProperties]
  renderers [{rendererClass=&NiagaraSpriteRendererProperties}]
 }
 Trail {
  emitterUpdateScript [
   &EmitterLifeCycle(MaxLoopCount=1, NextLoopDuration=3.5) {enabled=false}
   &EmitterState()
   &SpawnRate(
    SpawnRate=&MakeCustomFloatFromBool(
     Bool=/Niagara/DynamicInputs/Bool/SetBoolByFloatComparison(A:NiagaraFloat=$Emitter.LoopedAge, B:NiagaraFloat=2, ComparisonType:ENiagaraBooleanLogicOps=::NewEnumerator0)
     True Float=0
     False Float=10
    )
    Use Spawn Probability=false
   )
  ]
  particleSpawnScript [
   &InitializeRibbon(
    Lifetime=1.5
    Module.Write Color=false
    Module.Write Lifetime=true
    Module.Write Mass=true
    Module.Write Material Random=true
    Module.Write Position=false
    Module.Write RibbonTwist=false
    Module.Write RibbonWidth=true
    Position:Vector3f=&SimulationPosition()
    Ribbon Width=10
   )
   &AddVelocity(Velocity=<0 0 50>)
   /SetVariables(Particles.Acceleration:Vector3f=&UniformRangedVector(Minimum=<15 15 25>, Maximum=<-15 -15 25>)) {isSetParameters=true}
  ]
  particleUpdateScript [
   &ParticleState() {module=UpdateAge}
   &AccelerationForce(Acceleration=$Particles.Acceleration)
   /Niagara/Modules/Update/Forces/DragForce()
   /SetVariables(Particles.RibbonWidth=&Multiply_Float(
     A=$Particles.Initial.RibbonWidth
     B=&FloatFromCurve(FloatCurve=DataInterface() {propertyValues {Curve {keys [<at -0.0046 1.0113>,<at 1 1.5>]}}})
    )) {isSetParameters=true}
   /Niagara/Modules/Update/Material/DynamicMatParams(
    Module.Param0WriteEnabled=true
    Module.Param1WriteEnabled=true
    Module.Param2WriteEnabled=true
    Module.Param3WriteEnabled=true
    Param0Value:NiagaraFloat=&FloatFromCurve(FloatCurve=DataInterface() {propertyValues {Curve {keys [<at 0 -6>,<at 2.5 -8.4>]}}}, CurveIndex=$Emitter.LoopedAge)
    Param1Value=81
    Param2Value=50
    Param3Value=0.1
   )
   &Color(Scale Alpha=&FloatFromCurve(
     FloatCurve=DataInterface() {propertyValues {Curve {keys [
         <at 0.06 0 RCIM_Cubic>
         {arriveTangent=1.3731, interpMode=RCIM_Cubic, leaveTangent=1.3731, tangentMode=RCTM_User, time=0.5207, value=1.8105}
         <at 1.0011 0.002 RCIM_Cubic>
        ]}}}
     Scale Curve=&FloatFromCurve(
      FloatCurve=DataInterface() {propertyValues {Curve {keys [{arriveTangent=0.0043, interpMode=RCIM_Cubic, leaveTangent=0.0043, tangentMode=RCTM_User, time=-0.0046, value=1.0113}, <at 0.94 0>]}}}
      CurveIndex=$Emitter.NormalizedLoopAge
     )
    ))
   &SolveForcesAndVelocity(Clamp Velocity=false, Limit Acceleration=false)
  ]
  rendererClasses [&NiagaraRibbonRendererProperties]
  renderers [{rendererClass=&NiagaraRibbonRendererProperties}]
 }
}
summary {
 systemName=NS_Emote_HotStuff
 emitters [
  {emitterName=Trail, rendererClasses [&NiagaraRibbonRendererProperties]}
  {emitterName=Streaks2, rendererClasses [&NiagaraSpriteRendererProperties]}
 ]
 userVariables []
}
